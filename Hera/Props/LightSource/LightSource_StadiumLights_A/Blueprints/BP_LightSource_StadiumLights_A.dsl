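; Adjustable stadium light prop with a rod, a light fixture, instanced brackets and lamps, and a spotlight.
; The construction script clamps the fixture angle, aligns parts to the base mesh, yaws the assembly, tilts lamp instances in a batch update, then places the spotlight at a fixture socket plus offset.
; On server death it destroys the bracket and lamp instance components.
; Bracket-angle and minimum-height helpers are defined but not called here.
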(event EventBeginPlay)

(event Collision|EventActorBeginOverlap
  (params (OtherActor :type Actor%20Object%20Reference)))

(event EventTick
  (params (DeltaSeconds :type Float%20%28single-precision%29)))

(event EventOnDeathServer
  (params (Damage :type Float%20%28single-precision%29) (DamageTags :type Gameplay%20Tag%20Container%20Structure%20%28by%20ref%29) (Momentum :type Vector :default "0, 0, 0") (HitInfo :type Hit%20Result%20Structure%20%28by%20ref%29) (InstigatedBy :type Controller%20Object%20Reference) (DamageCauser :type Actor%20Object%20Reference) (EffectContext :type Gameplay%20Effect%20Context%20Handle%20Structure))
  (for _array_element (Utilities|Array|MakeArray (Variables|Default|GetISM_Brackets) (Variables|Default|GetISM_Lights))
    (Components|DestroyComponent _array_element)))

(fn CalculateMinLength
  (bind (_min _max) (Components|StaticMesh|GetLocalBounds (Variables|AdjustableLight|GetLightFixture)))
  (bind (_min_2 _max_2) (Components|StaticMesh|GetLocalBounds (Variables|AdjustableLight|GetUpRod)))
  (bind (_min_3 _max_3) (Components|StaticMesh|GetLocalBounds (Variables|Debug|GetStaticMeshComponent)))
  (Variables|Default|SetMinActorHeight (Utilities|Operators|Add (+ (.z _max_3) (Variables|AdjustableLight|GetBaseMeshBoundsOffset)) (.z _max) (+ (.z _max_2) (Variables|AdjustableLight|GetRodMeshOffset)))))

(fn SetLightAssemblyAngle
  (Transformation|SetWorldTransform (Variables|AdjustableLight|GetLightFixture) (Math|Transform|MakeTransform (.location (Transformation|GetWorldTransform (Variables|AdjustableLight|GetLightFixture))) (Math|Rotator|CombineRotators (.rotation (Transformation|GetWorldTransform (Variables|AdjustableLight|GetLightFixture))) (Math|Rotator|MakeRotator (.roll (.rotation (Transformation|GetWorldTransform (Variables|AdjustableLight|GetLightFixture)))) (.pitch (.rotation (Transformation|GetWorldTransform (Variables|AdjustableLight|GetLightFixture)))) (Variables|AdjustableLight|GetLightAssemblyAngle))) (.scale (Transformation|GetWorldTransform (Variables|AdjustableLight|GetLightFixture)))) false false))

(fn UpdateComponentRotation
  (bind (_returnvalue_location _returnvalue_rotation) (Math|Transform|BreakTransform (Transformation|GetWorldTransform (Variables|Debug|GetStaticMeshComponent))))
  (for _array_element (Utilities|Array|MakeArray (Variables|AdjustableLight|GetLightFixture) (Variables|Default|GetISM_Lights) (Variables|Default|GetISM_Brackets) (Variables|AdjustableLight|GetUpRod))
    (Transformation|SetWorldRotation _array_element _returnvalue_rotation false false)))

(fn UpdateLightBracketAngle
  (bind _outtransforms (Components|InstancedStaticMesh|GetInstanceTransforms (Variables|Default|GetISM_Brackets)))
  (for _array_element _outtransforms
    (Utilities|Array|Add (Variables|Default|GetAR_BracketTransforms) (Math|Transform|MakeTransform (.location _array_element) (Math|Rotator|MakeRotator (.roll (.rotation _array_element)) (.pitch (.rotation _array_element)) (Variables|AdjustableLight|GetLightFixtureAngle)) (.scale _array_element))))
  (Components|InstancedStaticMesh|BatchUpdateInstancesTransforms (Variables|Default|GetISM_Brackets) 0 (Variables|Default|GetAR_BracketTransforms) false false false))

(fn UpdateLightFixtureAngle
  (LocalVariables|SetL_InvertedFixtureAngle (* (Math|Float|ClampAngle (Variables|AdjustableLight|GetLightFixtureAngle) -45.0 45.0) -1.0))
  (bind _outtransforms (Components|InstancedStaticMesh|GetInstanceTransforms (Variables|Default|GetISM_Lights)))
  (for _array_element _outtransforms
    (Utilities|Array|Add (Variables|Default|GetAR_LightTransforms) (Math|Transform|MakeTransform (.location _array_element) (Math|Rotator|CombineRotators (.rotation _array_element) (Math|Rotator|MakeRotator (LocalVariables|GetL_InvertedFixtureAngle) (.pitch (.rotation _array_element)) (.yaw (.rotation _array_element)))))))
  (Components|InstancedStaticMesh|BatchUpdateInstancesTransforms (Variables|Default|GetISM_Lights) 0 (Variables|Default|GetAR_LightTransforms) false false false))

(fn UpdateLightTransforms
  (bind (_returnvalue_location _returnvalue_rotation _returnvalue_scale) (Math|Transform|BreakTransform (Transformation|GetSocketTransform (Variables|AdjustableLight|GetLightFixture) (Utilities|Array|Get(acopy) (Transformation|GetAllSocketNames (Variables|AdjustableLight|GetLightFixture))) "RTS_World")))
  (bind (_returnvalue_roll _returnvalue_pitch _returnvalue_yaw) (Math|Rotator|BreakRotator (Transformation|GetWorldRotation (Variables|AdjustableLight|GetLightFixture))))
  (Transformation|SetWorldTransform (Variables|Default|GetSpotLight) (Math|Transform|MakeTransform (+ _returnvalue_location (Variables|AdjustableLight|GetLightOffset)) (Math|Rotator|CombineRotators (Math|Rotator|MakeRotator 0.0 (Variables|AdjustableLight|GetLightFixtureAngle) 9e+01) (Math|Rotator|MakeRotator _returnvalue_roll 0.0 _returnvalue_yaw)) _returnvalue_scale) false false))

(fn ConstructionScript
  (|CallParentFunction)
  (Variables|AdjustableLight|SetLightFixtureAngle (Utilities|ClampMaxFloat (Utilities|ClampMinFloat (Variables|AdjustableLight|GetLightFixtureAngle) -3e+01) 3e+01))
  (CallFunction|UpdateComponentRotation)
  (CallFunction|SetLightAssemblyAngle)
  (CallFunction|UpdateLightFixtureAngle)
  (CallFunction|UpdateLightTransforms))
